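# Niagara one-shot impact effect for the Great Pool emote, apparently a bell strike with a fading ring-out.
# Flash star, light-emitting flare, blue sparks and impact smoke spawn at System.BellImpactPos; glow and secondary smoke follow a skeletal mesh bone or socket.
# Mesh ripple emitters add expanding rings, two driven by the RingOutTimeline for spawn rate and fade.
# User.Intensity scales ripples, front-end checks retune size and brightness, and spike and debug emitters are disabled.
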NiagaraSystem /BRCosmetics/Animation/Game/MainPlayer/Emotes/GreatPool/VFX/NS_Emote_Greatpool_Impact
dependencies {
 usedDataInterfaces [
  &NiagaraDataInterfaceCurve
  &NiagaraDataInterfaceVector2DCurve
  &NiagaraDataInterfaceColorCurve
  &NiagaraDataInterfaceMeshRendererInfo
  &NiagaraDataInterfaceVectorCurve
  &NiagaraDataInterfaceSkeletalMesh
 ]
 usedDynamicInputs [
  &FNTimeOfDayScalar
  &TransformPosition
  &Divide_Float
  &self:ScratchDynamicInput
 ]
 usedModules [
  &SampleSkeletalMesh
  &self:SystemSpawnScript.SystemScriptSource.SystemScriptGraph.NiagaraNodeAssignment_0.SetVariables_#1
  &self:SystemSpawnScript.SystemScriptSource.SystemScriptGraph.NiagaraNodeAssignment_4.SetVariables_#2
  &self:SystemSpawnScript.SystemScriptSource.SystemScriptGraph.NiagaraNodeAssignment_6.SetVariables_#3
  &SystemState
  &Timeline
  &self:SystemSpawnScript.SystemScriptSource.SystemScriptGraph.NiagaraNodeAssignment_2.SetVariables_#4
  &self:SystemSpawnScript.SystemScriptSource.SystemScriptGraph.NiagaraNodeAssignment_1.SetVariables_#5
  &self:SystemSpawnScript.SystemScriptSource.SystemScriptGraph.NiagaraNodeAssignment_5.SetVariables_#6
  &self:SystemSpawnScript.SystemScriptSource.SystemScriptGraph.NiagaraNodeAssignment_3.SetVariables_#7
 ]
}
emitters {
 AirGlow {
  emitterUpdateScript [
   &EmitterState()
   &SpawnBurst_Instantaneous(Spawn Group=1, Use Spawn Probability=false, UseLoopCountLimit=false) {enabled=false, module=SpawnBurst_Instantaneous001}
   &SpawnBurst_Instantaneous(Spawn Count=1, Use Spawn Probability=false)
  ]
  particleSpawnScript [
   &InitializeParticle(
    Color=<rgba 0.35 0.4323 0.4323 0.2>
    Lifetime=0.2
    Sprite Size Mode=::NewEnumerator3
    Uniform Sprite Size=250
    UsePositionOffset=false
    Write Material Random=true
   ) {module=InitializeParticle001}
   &DynamicMaterialParameters(
    Index 0 Param 1=2
    Index 0 Param 2=1
    Index 0 Param 3=1
    Index 0 Param 4=1
    Module.Param0WriteEnabled=true
    Module.Param1WriteEnabled=true
    Module.Param2WriteEnabled=true
    Module.Param3WriteEnabled=true
   )
   /SetVariables(Particles.SpawnGroup=$Emitter.SpawnGroup) {isSetParameters=true}
   &CameraOffset(Camera Offset Amount=-10)
   &SkeletalMeshLocation(
    Bone / Socket Index=2
    Interpolate Mesh Sampling=false
    MaskSpawnGroup=false
    Mesh Sampling Type=::NewEnumerator5
    Skeletal Mesh=$System.Skeletal Mesh
    Socket Sampling Mode=::NewEnumerator5
   )
  ]
  particleUpdateScript [
   &ParticleState()
   &ScaleSpriteSize(Scale Factor=&Vector2DFromFloat(Value=&FloatFromCurve(FloatCurve=DataInterface() {propertyValues {Curve {keys [<at 0 1 RCIM_Cubic>]}}})))
   &ScaleColor(
    Scale Alpha=&FloatFromCurve(FloatCurve=DataInterface() {propertyValues {Curve {keys [<at 0 0>,<at 0.25 0.75>,<at 1 0>]}}})
    ScaleA=true
    ScaleRGB=true
   ) {module=ScaleColor001}
  ]
  rendererClasses [&NiagaraSpriteRendererProperties]
  renderers [{rendererClass=&NiagaraSpriteRendererProperties}]
 }
 Default_Pickaxe_Impact_Smoke {
  emitterUpdateScript [
   &EmitterState()
   &SpawnBurst_Instantaneous(
    Spawn Count=&RandomRangeInt(Minimum=2, Maximum=3)
    Use Spawn Probability=false
    UseLoopCountLimit=false
   )
  ]
  particleSpawnScript [
   &InitializeParticle(
    Color Maximum=<rgba 0.76 1 1 0.2>
    Color Minimum=<rgba 0.8 0.917 1 0.15>
    Color Mode=::NewEnumerator2
    Lifetime Max=0.4
    Lifetime Min=0.3
    Lifetime Mode=::NewEnumerator1
    Position Offset=$System.BellImpactPos
    Sprite Size Mode=::NewEnumerator3
    Uniform Sprite Size=&Multiply_Float(A=0.3, B=&RandomRangeFloat(Minimum=100, Maximum=130))
    UsePositionOffset=true
    Write Material Random=true
   )
   &CameraOffset(Camera Offset Amount=&RandomRangeFloat(Minimum=40, Maximum=50, Evaluation Type=::NewEnumerator1))
   &ShapeLocation(
    Hemisphere Distribution=<1 0.5>
    MaskSpawnGroup=false
    Override Seed=false
    Rotation Angle=90
    Rotation Mode=::NewEnumerator1
    Sphere Radius=&MakeCustomFloatFromBool(Bool=$User.bIsFrontEnd, True Float=5, False Float=10)
    UseOwnerScale=true
   )
   &AddVelocity(Velocity=&Multiply_VectorByFloat(Vector=$Particles.ShapeLocation.ShapeVector, Float=10))
  ]
  particleUpdateScript [
   &ParticleState()
   &ScaleSpriteSize(Uniform Curve Sprite Scale=DataInterface() {propertyValues {Curve {keys [
        {arriveTangent=1.7761, interpMode=RCIM_Cubic, leaveTangent=1.7761, tangentMode=RCTM_User, time=0, value=0}
        {interpMode=RCIM_Cubic, tangentMode=RCTM_User, time=1, value=1}
       ]}}})
   &ScaleColor(
    Scale Alpha=&FloatFromCurve(FloatCurve=DataInterface() {propertyValues {Curve {keys [
         {arriveTangent=2.4358, interpMode=RCIM_Cubic, leaveTangent=2.4358, tangentMode=RCTM_User, time=0, value=0}
         <at 0.25 0.5 RCIM_Cubic>
         <at 1 0 RCIM_Cubic>
        ]}}})
    Scale RGB=&VectorFromCurve(VectorCurve=DataInterface() {propertyValues {XCurve {keys [<at 0 1>,<at 1 1>]}, YCurve {keys [<at 0 1>,<at 1 1>]}, ZCurve {keys [<at 0 1>,<at 1 1>]}}})
    ScaleA=true
    ScaleRGB=true
   )
   &DynamicMaterialParameters(
    Index 0 Param 1=&Add_Float(
     A=&RandomRangeFloat(Minimum=0, Maximum=1)
     B=&FloatFromCurve(
      FloatCurve=DataInterface() {propertyValues {Curve {keys [
          <at 0 0 RCIM_Cubic>
          {arriveTangent=1.5947, interpMode=RCIM_Cubic, leaveTangent=1.5947, tangentMode=RCTM_User, time=0.2557, value=0.5257}
          <at 1 1 RCIM_Cubic>
         ]}}}
      CurveIndex=$Particles.Age
      Scale Curve=0.25
     )
    )
    Module.Param0WriteEnabled=true
    Module.Param1WriteEnabled=true
    Module.Param2WriteEnabled=true
    Module.Param3WriteEnabled=true
   ) {module=DynamicMaterialParameters001}
   &InheritSourceMovement(ApplyForce=false, ApplyPosition=true, ApplyPositionFalloff=false) {enabled=false}
   &SolveForcesAndVelocity(Clamp Velocity=false, Limit Acceleration=false)
  ]
  rendererClasses [&NiagaraSpriteRendererProperties]
  renderers [
   {rendererClass=&NiagaraSpriteRendererProperties}
   {rendererClass=&NiagaraSpriteRendererProperties, rendererIndex=1}
  ]
 }
 FN_Debug {
  bEnabled=false
  emitterUpdateScript [
   &EmitterState()
   &SpawnBurst_Instantaneous(Spawn Count=1, Use Spawn Probability=false)
  ]
  particleSpawnScript [/Niagara/Modules/Spawn/Initialization/InitializeParticle(
    Lifetime=$System.CurrentLoopDuration
    Module.Write Color=true
    Module.Write Lifetime=true
    Module.Write Mass=true
    Module.Write Material Random=true
    Module.Write Position=false
    Module.Write Scale=false
    Module.Write SpriteRotation=false
    Module.Write SpriteSize=true
    Position=&SimulationPosition()
    Sprite Size=<100 100>
   )]
  particleUpdateScript [
   &ParticleState()
   /SetVariables(Particles.Debug:NiagaraFloat=$System.RingOutTimeline) {isSetParameters=true}
   &DynamicMaterialParameters(
    Index 0 Param 1=$Particles.Debug
    Module.Param0WriteEnabled=true
    Module.Param1WriteEnabled=true
    Module.Param2WriteEnabled=true
    Module.Param3WriteEnabled=true
   )
  ]
  rendererClasses [&NiagaraSpriteRendererProperties]
  renderers [{rendererClass=&NiagaraSpriteRendererProperties}]
 }
 Flash_Flare {
  emitterUpdateScript [
   &EmitterState()
   &SpawnBurst_Instantaneous(Use Spawn Probability=false, UseLoopCountLimit=false)
   &SpawnBurst_Instantaneous(Spawn Time=0.0531, Use Spawn Probability=false, UseLoopCountLimit=false) {enabled=false, module=SpawnBurst_Instantaneous001}
  ]
  particleSpawnScript [
   /Niagara/Modules/Spawn/Initialization/V2/InitializeParticle(
    Color=<rgba 0 0.2618 1 1>
    Lifetime=0.2
    Position Offset=$System.BellImpactPos
    Sprite Size=<100 100>
    Sprite Size Mode=::NewEnumerator4
    UsePositionOffset=true
    Write Material Random=true
   )
   &DynamicMaterialParameters(
    Index 0 Param 1=0.2
    Index 0 Param 2=1
    Module.Param0WriteEnabled=true
    Module.Param1WriteEnabled=true
    Module.Param2WriteEnabled=true
    Module.Param3WriteEnabled=true
   )
   &SystemLocation(MaskSpawnGroup=false, Offset=$System.BellImpactPos, Offset Coordinate Space=::World) {enabled=false}
  ]
  particleUpdateScript [
   &ParticleState()
   &ScaleColor(
    Scale Alpha=&Lerp_Float(A=1, B=0, Alpha=$Particles.NormalizedAge)
    Scale RGB=&VectorFromFloat(Value=&FN_FrontEndFloatModifier(Gameplay=1.5, Front End=1))
    ScaleA=true
    ScaleRGB=true
   )
   &FN_Light_Attributes(
    Color=$Particles.Color
    Create Color Attribute=true
    Create Exponent Attribute=true
    Create Radius Attribute=true
    Exponent=1
    Radius=100
    Scale Color=&FloatFromCurve(
     FloatCurve=DataInterface() {propertyValues {Curve {keys [
         {interpMode=RCIM_Cubic, tangentMode=RCTM_User, time=0, value=1}
         {arriveTangent=-0.5426, interpMode=RCIM_Cubic, leaveTangent=-0.5426, tangentMode=RCTM_User, time=1, value=0}
        ]}}}
     Scale Curve=4
    )
   )
   &CameraOffset()
   &InheritSourceMovement(ApplyForce=false, ApplyPosition=true, ApplyPositionFalloff=false) {enabled=false}
   &SolveForcesAndVelocity(Clamp Velocity=false, Limit Acceleration=false)
  ]
  rendererClasses [&NiagaraSpriteRendererProperties, &NiagaraLightRendererProperties]
  renderers [
   {rendererClass=&NiagaraSpriteRendererProperties}
   {rendererClass=&NiagaraLightRendererProperties, rendererIndex=1}
  ]
 }
 Flash_Star {
  emitterUpdateScript [
   &EmitterState(
    Enable Distance Culling=true
    Max Distance Response=::KillImmediately
    Module.bUseMaxDistance=true
    Module.bUseMinDistance=false
    Scalability Mode=::NewEnumerator1
   )
   &SpawnBurst_Instantaneous(Use Spawn Probability=false, UseLoopCountLimit=false)
  ]
  particleSpawnScript [
   &InitializeParticle(
    Lifetime=0.15
    Position Offset=$System.BellImpactPos
    Sprite Rotation Mode=::NewEnumerator2
    Sprite Size Mode=::NewEnumerator3
    Uniform Sprite Size=&Multiply_Float(
     A=&MakeCustomFloatFromBool(Bool=$User.bIsFrontEnd, True Float=0.4, False Float=0.4)
     B=&RandomRangeFloat(Minimum=120, Maximum=150)
    )
    UsePositionOffset=true
    Write Material Random=true
   )
   &FN_PerParticleLOD(Enable SpriteSize Distance Scaling=true, Sprite Size Scale Far=<2.5 2.5>)
   &CameraOffset(Camera Offset Amount=15)
   /SetVariables(Particles.RandomFloat=&UniformRangedFloat(Minimum=0, Maximum=1)) {isSetParameters=true}
   /SetVariables(
    Particles.SubImageIndex=&RandomRangeFloat(Minimum=0, Maximum=3)
    Particles.SubImageIndexAlt:NiagaraFloat=&RandomRangeFloat(Minimum=0, Maximum=3)
   ) {isSetParameters=true}
   /SetVariables(Particles.SpriteSizeAlt:Vector2f=&Multiply_Vector2D_ByFloat(Float=0.85, Vector2D=$Particles.SpriteSize)) {isSetParameters=true}
  ]
  particleUpdateScript [
   &ParticleState()
   &ScaleColor(
    Linear Color Curve=DataInterface() {propertyValues {
      AlphaCurve {keys [<at 0 1>,<at 0.8992 0>]}
      BlueCurve {keys [<at 0.1978 3>,<at 0.4068 2>,<at 0.6127 1.5>]}
      GreenCurve {keys [<at 0.1978 0.8409>,<at 0.4068 0.6969>,<at 0.6127 0.2281>]}
      RedCurve {keys [<at 0.1978 0.1494>,<at 0.4068 0>,<at 0.6127 0>]}
     }}
    Scale Mode=::NewEnumerator2
   )
   &ScaleSpriteSize(Uniform Curve Sprite Scale=DataInterface() {propertyValues {Curve {keys [{arriveTangent=2.6565, interpMode=RCIM_Cubic, leaveTangent=2.6565, tangentMode=RCTM_User, time=0, value=0}, <at 1 1 RCIM_Cubic>]}}}) {module=ScaleSpriteSize001}
  ]
  rendererClasses [&NiagaraSpriteRendererProperties]
  renderers [
   {rendererClass=&NiagaraSpriteRendererProperties}
   {rendererClass=&NiagaraSpriteRendererProperties, rendererIndex=1}
  ]
 }
 Flask_Spikes {
  bEnabled=false
  emitterUpdateScript [
   &EmitterState()
   &SpawnBurst_Instantaneous(
    Spawn Count=&RandomRangeInt(Minimum=6, Maximum=8)
    Use Spawn Probability=false
    UseLoopCountLimit=false
   )
  ]
  particleSpawnScript [
   &InitializeParticle(
    Lifetime Max=0.07
    Lifetime Min=0.05
    Lifetime Mode=::NewEnumerator1
    Position=$System.BellImpactPos
    Position Mode=::NewEnumerator1
    Sprite Size=&Multiply_Vector2D_ByFloat(
     Float=&MakeCustomFloatFromBool(Bool=$User.bIsFrontEnd, True Float=0.45, False Float=1)
     Vector2D=&RandomRangeVector2D(Minimum=<3 55>, Maximum=<10 70>)
    )
    Sprite Size Mode=::NewEnumerator4
    UsePositionOffset=false
    Write Material Random=true
   )
   &ShapeLocation(MaskSpawnGroup=false, Override Seed=false, Sphere Radius=1, UseOwnerScale=true)
   &SpriteFacingAndAlignment(Sprite Alignment=$Particles.ShapeLocation.ShapeNormal, WriteAlignment=true, WriteFacing=false)
  ]
  particleUpdateScript [
   &ParticleState()
   &ScaleColor(
    Scale Alpha=&FloatFromCurve(FloatCurve=DataInterface() {propertyValues {Curve {keys [<at 0.7 1>,<at 1 0>]}}})
    ScaleA=true
    ScaleRGB=false
   )
   &ScaleSpriteSize(Uniform Curve Sprite Scale=DataInterface() {propertyValues {Curve {keys [{tangentMode=RCTM_User, time=0, value=0.6}, {tangentMode=RCTM_User, time=1, value=2}]}}})
  ]
  rendererClasses [&NiagaraSpriteRendererProperties]
  renderers [{rendererClass=&NiagaraSpriteRendererProperties}]
 }
 ImpactRipples {
  emitterUpdateScript [
   &EmitterState()
   &SpawnBurst_Instantaneous(Use Spawn Probability=false, UseLoopCountLimit=false)
   &SpawnRate(
    SpawnRate=&FloatFromCurve(
     FloatCurve=DataInterface() {propertyValues {Curve {keys [<at 0 8 RCIM_Cubic>, {arriveTangent=-8, leaveTangent=-8, time=0.1, value=0.1}, <at 1 0>]}}}
     CurveIndex=$Emitter.NormalizedLoopAge
     Scale Curve=2
    )
    Use Spawn Probability=false
   ) {enabled=false}
  ]
  particleSpawnScript [
   &InitializeParticle(
    Color Maximum=<rgba 0.885 0.9091 1 0.02>
    Color Minimum=<rgba 0.759 0.8094 1 0.01>
    Color Mode=::NewEnumerator2
    Mesh Scale Mode=::NewEnumerator3
    Mesh Uniform Scale=0.8
    Position Offset=<0 0 -15.6481>
    Sprite Size Mode=::NewEnumerator3
    Uniform Sprite Size=200
    UsePositionOffset=true
    Write Material Random=true
   )
   &AddVelocity(Velocity=&RandomRangeVector(Minimum=<0 0 140>, Maximum=<1 1 160>))
   &InitialMeshOrientation(
    Mesh Orientation Mode=::NewEnumerator7
    Rotation=&RandomRangeVector(Minimum=<0 0 0>, Maximum=<3 0 0>)
    Use Rotation Vector=true
   )
  ]
  particleUpdateScript [
   &ParticleState()
   &UpdateMeshOrientation(Rotation Vector=<0 0 1>) {enabled=false}
   &ScaleSpriteSize(Uniform Curve Scale=&Lerp_Float(A=0.5, B=1, Alpha=$User.Intensity))
   &ScaleMeshSize(Scale Factor=&VectorFromFloat(Value=&FloatFromCurve(FloatCurve=DataInterface() {propertyValues {Curve {keys [
          {interpMode=RCIM_Cubic, tangentMode=RCTM_User, time=0, value=0}
          {arriveTangent=2, interpMode=RCIM_Cubic, leaveTangent=2, time=0.1, value=1.05}
          {interpMode=RCIM_Cubic, tangentMode=RCTM_User, time=1, value=2}
         ]}}})))
   &ScaleColor(
    Scale Alpha=&FloatFromCurve(
     FloatCurve=DataInterface() {propertyValues {Curve {keys [
         {arriveTangent=-3.5042, interpMode=RCIM_Cubic, leaveTangent=-3.5042, tangentMode=RCTM_User, time=0.1, value=1}
         {interpMode=RCIM_Cubic, tangentMode=RCTM_User, time=1, value=0}
        ]}}}
     Scale Curve=&FN_FrontEndFloatModifier(Gameplay=2, Front End=1.125)
    )
    ScaleA=true
    ScaleRGB=false
   )
   &Drag(Drag=7, Use Linear Drag=true, Use Rotational Drag=true)
   &SolveForcesAndVelocity(Clamp Velocity=false, Limit Acceleration=false)
   &DynamicMaterialParameters(
    Index 0 Param 1=&FloatFromCurve(FloatCurve=DataInterface() {propertyValues {Curve {keys [<at 0.1 5 RCIM_Cubic>,<at 0.95 0 RCIM_Cubic>]}}}, Scale Curve=0.25)
    Index 0 Param 2=&FloatFromCurve(FloatCurve=DataInterface() {propertyValues {Curve {keys [<at 0 5 RCIM_Cubic>,<at 0.4 36 RCIM_Cubic>]}}}, Scale Curve=0)
    Index 0 Param 3=1
    Index 0 Param 4=&FloatFromCurve(FloatCurve=DataInterface() {propertyValues {Curve {keys [<at 0 1 RCIM_Cubic>,<at 1 0 RCIM_Cubic>]}}})
    Module.Param0WriteEnabled=true
    Module.Param1WriteEnabled=true
    Module.Param2WriteEnabled=true
    Module.Param3WriteEnabled=true
   ) {enabled=false}
   &DynamicMaterialParameters(
    Index 0 Param 1=2
    Index 0 Param 2=1
    Index 0 Param 3=1
    Index 0 Param 4=&FloatFromCurve(FloatCurve=DataInterface() {propertyValues {Curve {keys [<at 0 1 RCIM_Cubic>, {arriveTangent=-1, interpMode=RCIM_Cubic, leaveTangent=-1, time=0.2, value=0.2}, <at 1 0 RCIM_Cubic>]}}})
    Module.Index0Param0Write=true
    Module.Index0Param1Write=true
    Module.Index0Param2Write=true
    Module.Index0Param3Write=true
   ) {module=DynamicMaterialParameters001}
  ]
  rendererClasses [&NiagaraSpriteRendererProperties, &NiagaraMeshRendererProperties]
  renderers [
   {rendererClass=&NiagaraSpriteRendererProperties}
   {rendererClass=&NiagaraMeshRendererProperties, rendererIndex=1}
   {rendererClass=&NiagaraMeshRendererProperties, rendererIndex=2}
   {rendererClass=&NiagaraMeshRendererProperties, rendererIndex=3}
   {rendererClass=&NiagaraMeshRendererProperties, rendererIndex=4}
  ]
 }
 MeshRipples {
  emitterUpdateScript [
   &EmitterState()
   &SpawnBurst_Instantaneous(Use Spawn Probability=false, UseLoopCountLimit=false)
   &SpawnRate(
    SpawnRate=&FloatFromCurve(
     FloatCurve=DataInterface() {propertyValues {Curve {keys [<at 0 8 RCIM_Cubic>, {arriveTangent=-4, interpMode=RCIM_Cubic, leaveTangent=-4, time=0.2, value=5.25}, <at 1 4 RCIM_Cubic>]}}}
     CurveIndex=$System.RingOutTimeline
    )
    Use Spawn Probability=false
   )
  ]
  particleSpawnScript [
   &InitializeParticle(
    Color Maximum=<rgba 0.885 0.9091 1 0.02>
    Color Minimum=<rgba 0.759 0.8094 1 0.01>
    Color Mode=::NewEnumerator2
    Mesh Scale Mode=::NewEnumerator3
    Position Offset=<0 0 -15.6481>
    Sprite Size Mode=::NewEnumerator3
    Uniform Sprite Size=100
    UsePositionOffset=true
    Write Material Random=true
   )
   &AddVelocity(Velocity=&RandomRangeVector(Minimum=<0 0 140>, Maximum=<1 1 160>))
   &InitialMeshOrientation(
    Mesh Orientation Mode=::NewEnumerator7
    Rotation=&RandomRangeVector(Minimum=<0 0 0>, Maximum=<3 0 0>)
    Use Rotation Vector=true
   )
  ]
  particleUpdateScript [
   &ParticleState()
   &UpdateMeshOrientation(Rotation Vector=<0 0 1>) {enabled=false}
   &ScaleSpriteSize(Uniform Curve Scale=&Lerp_Float(A=0.5, B=1, Alpha=$User.Intensity))
   &ScaleMeshSize(Scale Factor=&VectorFromFloat(Value=&FloatFromCurve(FloatCurve=DataInterface() {propertyValues {Curve {keys [{interpMode=RCIM_Cubic, tangentMode=RCTM_User, time=0, value=0.2}, {interpMode=RCIM_Cubic, tangentMode=RCTM_User, time=1, value=1}]}}})))
   &ScaleColor(
    Scale Alpha=&FloatFromCurve(
     FloatCurve=DataInterface() {propertyValues {Curve {keys [
         {arriveTangent=9.6668, interpMode=RCIM_Cubic, time=0, value=0}
         <at 0.2 1 RCIM_Cubic>
         {interpMode=RCIM_Cubic, tangentMode=RCTM_User, time=1, value=0}
        ]}}}
     Scale Curve=&FloatFromCurve(
      FloatCurve=DataInterface() {propertyValues {Curve {keys [<at 0 1 RCIM_Cubic>,<at 1 0.25 RCIM_Cubic>]}}}
      CurveIndex=$System.RingOutTimeline
      Scale Curve=&FN_FrontEndFloatModifier(Gameplay=2.5, Front End=1.5)
     )
    )
    ScaleA=true
    ScaleRGB=false
   )
   &Drag(Drag=7, Use Linear Drag=true, Use Rotational Drag=true)
   &SolveForcesAndVelocity(Clamp Velocity=false, Limit Acceleration=false)
   &DynamicMaterialParameters(
    Index 0 Param 1=&FloatFromCurve(FloatCurve=DataInterface() {propertyValues {Curve {keys [<at 0.1 5 RCIM_Cubic>,<at 0.95 0 RCIM_Cubic>]}}}, Scale Curve=0.25)
    Index 0 Param 2=&FloatFromCurve(FloatCurve=DataInterface() {propertyValues {Curve {keys [<at 0 5 RCIM_Cubic>,<at 0.4 36 RCIM_Cubic>]}}}, Scale Curve=0)
    Index 0 Param 3=1
    Index 0 Param 4=&FloatFromCurve(FloatCurve=DataInterface() {propertyValues {Curve {keys [<at 0 1 RCIM_Cubic>,<at 1 0 RCIM_Cubic>]}}})
    Module.Param0WriteEnabled=true
    Module.Param1WriteEnabled=true
    Module.Param2WriteEnabled=true
    Module.Param3WriteEnabled=true
   )
   &KillParticles(Kill Particles=$System.ShouldRemoveAll) {enabled=false}
  ]
  rendererClasses [&NiagaraSpriteRendererProperties, &NiagaraMeshRendererProperties]
  renderers [
   {rendererClass=&NiagaraSpriteRendererProperties}
   {rendererClass=&NiagaraMeshRendererProperties, rendererIndex=1}
  ]
 }
 MeshRipples001 {
  emitterUpdateScript [
   &EmitterState()
   &SpawnBurst_Instantaneous(Use Spawn Probability=false, UseLoopCountLimit=false)
   &SpawnRate(
    SpawnRate=&FloatFromCurve(
     FloatCurve=DataInterface() {propertyValues {Curve {keys [<at 0 8 RCIM_Cubic>, {arriveTangent=-4, interpMode=RCIM_Cubic, leaveTangent=-4, time=0.1, value=6}, <at 1 4 RCIM_Cubic>]}}}
     CurveIndex=$System.RingOutTimeline
    )
    Use Spawn Probability=false
   )
  ]
  particleSpawnScript [
   &InitializeParticle(
    Color Maximum=<rgba 1 0.9348 0.83 0.07>
    Color Minimum=<rgba 1 0.8667 0.796 0.03>
    Color Mode=::NewEnumerator2
    Mesh Scale Mode=::NewEnumerator3
    Mesh Uniform Scale=135
    Sprite Size Mode=::NewEnumerator3
    Uniform Sprite Size=100
    UsePositionOffset=false
    Write Material Random=true
   )
   &AddVelocity(Velocity=&RandomRangeVector(Minimum=<0 0 1>, Maximum=<1 1 1>))
   &InitialMeshOrientation(
    Mesh Orientation Mode=::NewEnumerator1
    Rotation=&RandomRangeVector(Minimum=<0 0 0>, Maximum=<1 0 0>)
    Use Rotation Vector=true
   )
  ]
  particleUpdateScript [
   &ParticleState()
   &UpdateMeshOrientation(Rotation Rate=0.2, Rotation Vector=<0 0 1>) {enabled=false}
   &ScaleSpriteSize(Uniform Curve Scale=&Lerp_Float(A=0.5, B=1, Alpha=$User.Intensity)) {enabled=false}
   &ScaleMeshSize(Scale Factor=&VectorFromFloat(Value=&FloatFromCurve(
      FloatCurve=DataInterface() {propertyValues {Curve {keys [{tangentMode=RCTM_User, time=0, value=0}, {tangentMode=RCTM_User, time=1, value=1}]}}}
      Scale Curve=&Lerp_Float(A=0.5, B=1, Alpha=$User.Intensity)
     )))
   &ScaleColor(
    Scale Alpha=&FloatFromCurve(
     FloatCurve=DataInterface() {propertyValues {Curve {keys [
         {arriveTangent=9.6668, interpMode=RCIM_Cubic, leaveTangent=9.6668, tangentMode=RCTM_User, time=0, value=0}
         <at 0.1 1 RCIM_Cubic>
         {interpMode=RCIM_Cubic, tangentMode=RCTM_User, time=1, value=0}
        ]}}}
     Scale Curve=&FloatFromCurve(
      FloatCurve=DataInterface() {propertyValues {Curve {keys [<at 0 1 RCIM_Cubic>, {arriveTangent=-0.75, interpMode=RCIM_Cubic, leaveTangent=-0.75, time=0.7, value=0.85}, <at 1 0.25 RCIM_Cubic>]}}}
      CurveIndex=$System.RingOutTimeline
      Scale Curve=&FN_FrontEndFloatModifier(Gameplay=2, Front End=1.5)
     )
    )
    ScaleA=true
    ScaleRGB=true
   )
   &Drag(Use Linear Drag=true, Use Rotational Drag=true)
   &SolveForcesAndVelocity(Clamp Velocity=false, Limit Acceleration=false)
   &DynamicMaterialParameters(
    Index 0 Param 1=0.5
    Index 0 Param 2=1
    Index 0 Param 3=1
    Index 0 Param 4=1
    Module.Param0WriteEnabled=true
    Module.Param1WriteEnabled=true
    Module.Param2WriteEnabled=true
    Module.Param3WriteEnabled=true
   )
  ]
  rendererClasses [&NiagaraSpriteRendererProperties, &NiagaraMeshRendererProperties]
  renderers [
   {rendererClass=&NiagaraSpriteRendererProperties}
   {rendererClass=&NiagaraMeshRendererProperties, rendererIndex=1}
  ]
 }
 MeshRipples002 {
  emitterUpdateScript [
   &EmitterState(Life Cycle Mode=::NewEnumerator1, Loop Behavior=::NewEnumerator1, Loop Duration=0.3, UseLoopDelay=false)
   &SpawnBurst_Instantaneous(Use Spawn Probability=false, UseLoopCountLimit=false)
   &SpawnRate(
    SpawnRate=&FloatFromCurve(
     FloatCurve=DataInterface() {propertyValues {Curve {keys [<at 0 8 RCIM_Cubic>, {arriveTangent=-4, interpMode=RCIM_Cubic, leaveTangent=-4, time=0.1, value=6}, <at 1 4 RCIM_Cubic>]}}}
     CurveIndex=$Emitter.NormalizedLoopAge
     Scale Curve=1.5
    )
    Use Spawn Probability=false
   )
  ]
  particleSpawnScript [
   &InitializeParticle(
    Color Maximum=<rgba 0.885 0.9091 1 0.02>
    Color Minimum=<rgba 0.759 0.8094 1 0.01>
    Color Mode=::NewEnumerator2
    Lifetime=0.666
    Mesh Scale Mode=::NewEnumerator3
    Mesh Uniform Scale=0.75
    Position Offset=<0 0 -15.6481>
    Sprite Size Mode=::NewEnumerator3
    Uniform Sprite Size=75
    UsePositionOffset=true
    Write Material Random=true
   )
   &AddVelocity(Velocity=&RandomRangeVector(Minimum=<0 0 140>, Maximum=<1 1 160>))
   &InitialMeshOrientation(
    Mesh Orientation Mode=::NewEnumerator7
    Rotation=&RandomRangeVector(Minimum=<0 0 0>, Maximum=<3 0 0>)
    Use Rotation Vector=true
   )
  ]
  particleUpdateScript [
   &ParticleState()
   &UpdateMeshOrientation(Rotation Vector=<0 0 1>) {enabled=false}
   &ScaleSpriteSize(Uniform Curve Scale=&Lerp_Float(A=0.5, B=1, Alpha=$User.Intensity))
   &ScaleMeshSize(Scale Factor=&VectorFromFloat(Value=&FloatFromCurve(FloatCurve=DataInterface() {propertyValues {Curve {keys [
          {arriveTangent=3.8683, interpMode=RCIM_Cubic, leaveTangent=3.8683, tangentMode=RCTM_User, time=0, value=0.3}
          {interpMode=RCIM_Cubic, tangentMode=RCTM_User, time=1, value=1.5}
         ]}}})))
   &ScaleColor(
    Scale Alpha=&FloatFromCurve(
     FloatCurve=DataInterface() {propertyValues {Curve {keys [<at 0.1 1 RCIM_Cubic>, {interpMode=RCIM_Cubic, tangentMode=RCTM_User, time=1, value=0}]}}}
     Scale Curve=0.7
    )
    ScaleA=true
    ScaleRGB=false
   )
   &Drag(Drag=7, Use Linear Drag=true, Use Rotational Drag=true)
   &SolveForcesAndVelocity(Clamp Velocity=false, Limit Acceleration=false)
   &DynamicMaterialParameters(
    Index 0 Param 1=&FloatFromCurve(FloatCurve=DataInterface() {propertyValues {Curve {keys [<at 0.1 5 RCIM_Cubic>,<at 0.95 0 RCIM_Cubic>]}}}, Scale Curve=0.25)
    Index 0 Param 2=&FloatFromCurve(FloatCurve=DataInterface() {propertyValues {Curve {keys [<at 0 5 RCIM_Cubic>,<at 0.4 36 RCIM_Cubic>]}}}, Scale Curve=-0.05)
    Index 0 Param 3=1
    Index 0 Param 4=&FloatFromCurve(FloatCurve=DataInterface() {propertyValues {Curve {keys [<at 0 1 RCIM_Cubic>,<at 1 0 RCIM_Cubic>]}}})
    Module.Param0WriteEnabled=true
    Module.Param1WriteEnabled=true
    Module.Param2WriteEnabled=true
    Module.Param3WriteEnabled=true
   )
  ]
  rendererClasses [&NiagaraSpriteRendererProperties, &NiagaraMeshRendererProperties]
  renderers [
   {rendererClass=&NiagaraSpriteRendererProperties}
   {rendererClass=&NiagaraMeshRendererProperties, rendererIndex=1}
  ]
 }
 Smoke {
  emitterUpdateScript [
   &EmitterState()
   &SpawnBurst_Instantaneous(Spawn Count=5, Spawn Group=1, Use Spawn Probability=false, UseLoopCountLimit=false)
  ]
  particleSpawnScript [
   &InitializeParticle(
    Color=<rgba 0.5255 0.5255 0.5255 0.2>
    Lifetime=&UniformRangedFloat(Minimum=0.4, Maximum=1)
    Sprite Rotation Angle=&UniformRangedFloat(Minimum=0, Maximum=1)
    Sprite Rotation Mode=::NewEnumerator1
    Sprite Size Mode=::NewEnumerator3
    Uniform Sprite Size=&UniformRangedFloat(Minimum=25, Maximum=35)
    UsePositionOffset=false
    Write Material Random=true
   ) {module=InitializeParticle001}
   /SetVariables(Particles.InitSpriteRotRate=&UniformRangedFloat(Minimum=-1, Maximum=1)) {isSetParameters=true}
   /SetVariables(Particles.SpawnGroup=$Emitter.SpawnGroup) {isSetParameters=true}
   &AccelerationForce(Acceleration=&UniformRangedVector(Minimum=<0 0 -25>, Maximum=<0 0 -75>))
   &SphereLocation(MaskSpawnGroup=false, Override Seed=false, Surface Only=true, UseNonUniformScale=false, UseOffset=true, UseSphereOrientation=false)
   &ApplyInitialForces()
   &CameraOffset(Camera Offset Amount=&RandomRangeFloat(Minimum=0, Maximum=0))
   &AddVelocity(Velocity=&Add_Vector(
     A=&Multiply_Vector(A=$Output.SphereLocation.SphereNormal, B=&VectorFromFloat(Value=50))
     B=<0 0 10>
    ))
   &SkeletalMeshLocation(
    Bone / Socket Index=2
    Interpolate Mesh Sampling=false
    MaskSpawnGroup=false
    Mesh Sampling Type=::NewEnumerator5
    Sampled Position Offset=<0 0 5>
    Skeletal Mesh=$System.Skeletal Mesh
    Socket Sampling Mode=::NewEnumerator5
   )
  ]
  particleUpdateScript [
   &ParticleState()
   &SpriteRotationRate(Rotation Rate=&Multiply_Float(
     A=$Particles.InitSpriteRotRate
     B=&FloatFromCurve(
      FloatCurve=DataInterface() {propertyValues {Curve {keys [
          {arriveTangent=-1.9214, interpMode=RCIM_Cubic, leaveTangent=-1.9214, tangentMode=RCTM_User, time=-0.0046, value=1.0113}
          {arriveTangent=-0.2496, interpMode=RCIM_Cubic, leaveTangent=-0.2496, tangentMode=RCTM_User, time=0.9896, value=0.16}
         ]}}}
      CurveIndex=$Emitter.Age
     )
    ))
   &ScaleColor(
    Scale Alpha=&FloatFromCurve(FloatCurve=DataInterface() {propertyValues {Curve {keys [<at 0 0>,<at 0.2049 0.9765>,<at 0.7791 0.9107>,<at 1 0>]}}})
    ScaleA=true
    ScaleRGB=true
   ) {module=ScaleColor001}
   &Drag(Drag=3, Use Linear Drag=true, Use Rotational Drag=false)
   &InheritSourceMovement(
    Applied Position Scale=&VectorFromFloat(Value=&FloatFromCurve(FloatCurve=DataInterface() {propertyValues {Curve {keys [<at 0.75 1 RCIM_Cubic>,<at 1 0 RCIM_Cubic>]}}}))
    ApplyForce=false
    ApplyPosition=true
    ApplyPositionFalloff=false
   )
   &SolveForcesAndVelocity(Clamp Velocity=false, Limit Acceleration=false)
   &ScaleSpriteSize(Scale Factor=&Vector2DFromFloat(Value=&FloatFromCurve(
      FloatCurve=DataInterface() {propertyValues {Curve {keys [
          {arriveTangent=5.2354, interpMode=RCIM_Cubic, leaveTangent=5.2354, tangentMode=RCTM_User, time=-0.0046, value=0.25}
          <at 0.1439 1.4923>
          <at 1.0011 3.4809 RCIM_Cubic>
         ]}}}
      Scale Curve=&MaskFloatBySpawnGroup(Masked Float=0.75, Passthrough Float=1, Spawn Group Mask=1)
     )))
   &DynamicMaterialParameters(
    Index 0 Param 1=&FloatFromCurve(FloatCurve=DataInterface() {propertyValues {Curve {keys [<at 0 0 RCIM_Cubic>,<at 1.0011 0.2>]}}})
    Index 0 Param 2=&FloatFromCurve(FloatCurve=DataInterface() {propertyValues {Curve {keys [
         {arriveTangent=0.0453, interpMode=RCIM_Cubic, tangentMode=RCTM_Break, time=0, value=-1}
         {arriveTangent=0.0604, interpMode=RCIM_Cubic, leaveTangent=0.0604, tangentMode=RCTM_User, time=1, value=-1}
        ]}}})
    Index 0 Param 3=&FloatFromCurve(FloatCurve=DataInterface() {propertyValues {Curve {keys [{interpMode=RCIM_Cubic, leaveTangent=4.6779, time=0.2851, value=0}]}}})
    Index 1 Param 1=&FloatFromCurve(
     FloatCurve=DataInterface() {propertyValues {Curve {keys [<at 0 0>,<at 0.25 1>]}}}
     Scale Curve=&RandomRangeFloat(Minimum=2, Maximum=1)
    )
    Index 1 Param 2=&FloatFromCurve(
     FloatCurve=DataInterface() {propertyValues {Curve {keys [
         {interpMode=RCIM_Cubic, leaveTangent=3.7422, tangentMode=RCTM_Break, time=0, value=0.2}
         {interpMode=RCIM_Cubic, leaveTangent=0.9066, time=1, value=2}
        ]}}}
     Scale Curve=&RandomRangeFloat(Minimum=0.5, Maximum=1)
    )
    Index 1 Param 3=&FloatFromCurve(FloatCurve=DataInterface() {propertyValues {Curve {keys [
         {interpMode=RCIM_Cubic, leaveTangent=2.8445, tangentMode=RCTM_Break, time=0, value=0.2}
         {interpMode=RCIM_Cubic, leaveTangent=63.2978, time=0.25, value=1}
        ]}}})
    Index 1 Param 4=&FloatFromCurve(
     FloatCurve=DataInterface() {propertyValues {Curve {keys [<at 0 1>,<at 1 0.8>]}}}
     Scale Curve=&RandomRangeFloat(Minimum=0.5, Maximum=1)
    )
    Index 2 Param 1=&FloatFromCurve(FloatCurve=DataInterface() {propertyValues {Curve {keys [{interpMode=RCIM_Cubic, leaveTangent=3.2309, tangentMode=RCTM_Break, time=0.1, value=0}, <at 1 1>]}}})
    Index 2 Param 2=$Particles.MaterialRandom
    Index 2 Param 3=&FloatFromCurve(FloatCurve=DataInterface() {propertyValues {Curve {keys [
         {interpMode=RCIM_Cubic, tangentMode=RCTM_Break, time=0, value=1}
         {interpMode=RCIM_Cubic, leaveTangent=11.1233, tangentMode=RCTM_Break, time=1, value=2}
        ]}}})
    Index 2 Param 4=1.2
    Module.Param0WriteEnabled=true
    Module.Param0WriteEnabled1=true
    Module.Param0WriteEnabled2=true
    Module.Param1WriteEnabled=false
    Module.Param1WriteEnabled1=true
    Module.Param1WriteEnabled2=true
    Module.Param2WriteEnabled=true
    Module.Param2WriteEnabled1=true
    Module.Param2WriteEnabled2=true
    Module.Param3WriteEnabled=true
    Module.Param3WriteEnabled1=true
    Module.Param3WriteEnabled2=true
    Write Parameter Index 1=true
    Write Parameter Index 2=true
   )
  ]
  rendererClasses [&NiagaraSpriteRendererProperties]
  renderers [{rendererClass=&NiagaraSpriteRendererProperties}]
 }
 Sparks {
  emitterUpdateScript [
   &EmitterState(
    Enable Distance Culling=true
    Max Distance Response=::KillImmediately
    Module.bApplySpawnCountScaleByDistance=true
    Module.bUseMaxDistance=true
    Module.bUseMinDistance=false
    Scalability Mode=::NewEnumerator1
    Scale Spawn Count=true
    Spawn Count Scale By Distance=DataInterface() {propertyValues {Curve {keys [<at 0.3676 0.9657>,<at 0.9962 0.4982>]}}}
   )
   &SpawnBurst_Instantaneous(Spawn Count=&UniformRangedInt(Minimum=5, Maximum=10), Use Spawn Probability=false)
   &SpawnRate(
    Spawn Probability=&FloatFromCurve(
     FloatCurve=DataInterface() {propertyValues {Curve {keys [{arriveTangent=-12.2872, interpMode=RCIM_Cubic, leaveTangent=-12.2872, tangentMode=RCTM_User, time=0, value=1}, <at 0.1 0>]}}}
     CurveIndex=$System.Age
    )
    SpawnRate=111
    Use Spawn Probability=true
   )
  ]
  particleSpawnScript [
   &InitializeParticle(
    Color=<rgba 3 3 3 1>
    Lifetime Max=0.45
    Lifetime Min=0.166
    Lifetime Mode=::NewEnumerator1
    Position Offset:NiagaraPosition=$System.BellImpactPos
    Sprite Size Mode=::NewEnumerator3
    Uniform Sprite Size=&Multiply_Float(
     A=&MakeCustomFloatFromBool(Bool=$User.bIsFrontEnd, True Float=0.45, False Float=1)
     B=&RandomRangeFloat(Minimum=2.5, Maximum=7)
    )
    UsePositionOffset=true
    Write Material Random=true
   )
   &ShapeLocation(MaskSpawnGroup=false, Override Seed=false, Sphere Radius=9, Sphere Surface Distribution=1, UseOwnerScale=true)
   &AddVelocity(
    Cone Angle=90
    Cone Axis=<-1 1 1>
    Module.UseConeFalloff=false
    Module.UseVelDistribution=false
    Override Seed=false
    Rotation Coordinate Space=::Simulation
    Velocity Mode=::NewEnumerator2
    Velocity Speed=&RandomRangeFloat(Minimum=111, Maximum=500)
   )
   &DynamicMaterialParameters(
    Index 0 Param 1=3
    Index 0 Param 2=20
    Index 0 Param 3=1
    Module.Param0WriteEnabled=true
    Module.Param1WriteEnabled=true
    Module.Param2WriteEnabled=true
    Module.Param3WriteEnabled=true
   )
  ]
  particleUpdateScript [
   &ParticleState()
   &GravityForce(Gravity=<0 0 -666>)
   &Drag(Drag=&UniformRangedFloat(Minimum=2, Maximum=6), Use Linear Drag=true, Use Rotational Drag=false)
   &SolveForcesAndVelocity(Clamp Velocity=false, Limit Acceleration=false)
   &ScaleSpriteSize(Scale Factor=&Vector2DFromCurve(
     Vector2Curve=DataInterface() {propertyValues {
       XCurve {keys [
         {arriveTangent=5.9728, interpMode=RCIM_Cubic, leaveTangent=5.9728, tangentMode=RCTM_User, time=0, value=0.1}
         {arriveTangent=-0.1, interpMode=RCIM_Cubic, leaveTangent=-0.1, tangentMode=RCTM_User, time=0.18, value=0.8}
         {interpMode=RCIM_Cubic, tangentMode=RCTM_User, time=1, value=0}
        ]}
       YCurve {keys [
         {arriveTangent=6.2715, interpMode=RCIM_Cubic, leaveTangent=6.2715, tangentMode=RCTM_User, time=0, value=0.1}
         {arriveTangent=-0.1001, interpMode=RCIM_Cubic, leaveTangent=-0.1001, tangentMode=RCTM_User, time=0.18, value=0.8}
         {interpMode=RCIM_Cubic, tangentMode=RCTM_User, time=0.9988, value=0}
        ]}
      }}
     ScaleCurve=<1 1>
    ))
   &ScaleSpriteSizeBySpeed(Max Scale Factor=<0.5 4>, Min Scale Factor=<0.5 0.5>, Velocity Threshold=666)
   &ScaleColor(
    Linear Color Curve=DataInterface() {propertyValues {
      AlphaCurve {keys [<at 0.333 1>,<at 1 0>]}
      BlueCurve {keys [<at 0.012 2>,<at 0.333 4>,<at 0.75 3>]}
      GreenCurve {keys [<at 0.012 0.9466>,<at 0.333 0.6722>,<at 0.75 0.5541>]}
      RedCurve {keys [<at 0.012 0.3699>,<at 0.333 0.0891>,<at 0.75 0>]}
     }}
    Scale Mode=::NewEnumerator2
   )
  ]
  rendererClasses [&NiagaraSpriteRendererProperties]
  renderers [{rendererClass=&NiagaraSpriteRendererProperties}]
 }
}
summary {
 systemName=NS_Emote_Greatpool_Impact
 emitters [
  {emitterName=AirGlow, rendererClasses [&NiagaraSpriteRendererProperties]}
  {emitterName=Smoke, rendererClasses [&NiagaraSpriteRendererProperties]}
  {bEnabled=false, emitterName=Flask_Spikes, rendererClasses [&NiagaraSpriteRendererProperties]}
  {emitterName=Flash_Star, rendererClasses [&NiagaraSpriteRendererProperties]}
  {emitterName=Sparks, rendererClasses [&NiagaraSpriteRendererProperties]}
  {emitterName=Flash_Flare, rendererClasses [&NiagaraSpriteRendererProperties, &NiagaraLightRendererProperties]}
  {emitterName=Default_Pickaxe_Impact_Smoke, rendererClasses [&NiagaraSpriteRendererProperties]}
  {emitterName=MeshRipples, rendererClasses [&NiagaraSpriteRendererProperties, &NiagaraMeshRendererProperties]}
  {emitterName=MeshRipples001, rendererClasses [&NiagaraSpriteRendererProperties, &NiagaraMeshRendererProperties]}
  {bEnabled=false, emitterName=FN_Debug, rendererClasses [&NiagaraSpriteRendererProperties]}
  {emitterName=ImpactRipples, rendererClasses [&NiagaraSpriteRendererProperties, &NiagaraMeshRendererProperties]}
  {emitterName=MeshRipples002, rendererClasses [&NiagaraSpriteRendererProperties, &NiagaraMeshRendererProperties]}
 ]
 userVariables [
  User.bIsFrontEnd {type=NiagaraBool}
  User.bIsFrontEndPreview {type=NiagaraBool}
  User.Hit {type=NiagaraFloat, defaultValue=1, description=""}
  User.HitIntensity {type=NiagaraFloat, defaultValue=1, description=""}
  User.Intensity {type=NiagaraFloat, defaultValue=1, description=""}
  User.PhysicalSurface {type=NiagaraInt32}
  User.RibbonIDOffset {type=NiagaraInt32}
 ]
}
